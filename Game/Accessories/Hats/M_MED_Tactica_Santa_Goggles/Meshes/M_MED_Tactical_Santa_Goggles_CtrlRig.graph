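# Control Rig for the M_MED Tactical Santa Goggles hat accessory that applies corrective offsets to beard bones, including left and right side beard bones, as the head rotates.
# A reusable DeltaRotation Z function takes the head's transform relative to spine_05, converts it to Euler, and remaps the Z angle to a weight.
# Beard_Sides_Setup feeds that weight into a Modify Transforms node, blending in the preset offsets.
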ControlRigBlueprint /Game/Accessories/Hats/M_MED_Tactica_Santa_Goggles/Meshes/M_MED_Tactical_Santa_Goggles_CtrlRig
graphs [
 New Function_ContainedGraph {nodes [
   RigVMFunctionEntryNode()
   > RigVMFunction_Sequence = RigVMAggregateNode()
   DeltaRotation Z = RigVMFunctionReferenceNode(Child=head, Parent=spine_05, SourceMax=15, SourceMin=2, ExecuteContext=@RigVMFunction_Sequence.A)
   > &RigUnit_ModifyTransforms(
    ItemToModify="((Item=(Type=Bone,Name=\"dyn_beard_01_l\"),Transform=(Rotation=(X=-0.1771,Y=0.4138,Z=0.0278,W=-0.8926),Translation=(X=0.15,Y=0,Z=-0.25),Scale3D=(X=1,Y=1,Z=1))),(Item=(Type=Bone,Name=\"dyn_beard_01_r\"),Transform=(Rotation=(X=0,Y=-0.0872,Z=0,W=0.9962),Translati ...(truncated)"
    WeightMinimum=0
    Weight=@"DeltaRotation Z".Delta Z
   )
   RigVMFunctionReturnNode(ExecuteContext=@RigVMFunction_Sequence.B)
  ]}
 New Function_ContainedGraph {nodes [
   n1 = RigVMFunctionEntryNode()
   RigVMFunction_MathQuaternionToEuler = &RigVMFunction_MathQuaternionToEuler(
    RotationOrder=ZYX
    Value=&RigUnit_GetRelativeTransformForItem(
     Child=(Type=Bone,Name="None")
     Parent=(Type=Bone,Name="None")
     bChildInitial=False
     bParentInitial=False
     Child.Name=@n1.Child
     Parent.Name=@n1.Parent
    ).RelativeTransform.Rotation
   )
   Remap = &RigVMFunction_MathDoubleRemap(
    TargetMaximum=1
    TargetMinimum=0
    Value=@RigVMFunction_MathQuaternionToEuler.Result.Z
    SourceMaximum=@n1.SourceMax
    SourceMinimum=@n1.SourceMin
   )
   > RigVMFunctionReturnNode(Delta Z=@Remap.Result)
  ]}
 RigVMFunctionLibrary {nodes [
   Beard_Sides_Setup = RigVMCollapseNode(ExecuteContext=())
   DeltaRotation Z = RigVMCollapseNode(ExecuteContext=())
  ]}
 RigVMFunction_Sequence_1_ContainedGraph {nodes [
   n1 = RigVMFunctionEntryNode()
   RigVMFunction_Sequence = &RigVMFunction_Sequence(ExecuteContext=@n1.ExecuteContext)
   n2 = &RigVMFunction_Sequence(ExecuteContext=@RigVMFunction_Sequence.B)
   n3 = &RigVMFunction_Sequence(ExecuteContext=@n2.B)
   n4 = &RigVMFunction_Sequence(ExecuteContext=@n3.B)
   RigVMFunctionReturnNode(A=@RigVMFunction_Sequence.A, B=@n2.A, C=@n3.A, D=@n4.A, E=@n4.B)
  ]}
 RigVMModel {nodes [
   RigUnit_BeginExecution = &RigUnit_BeginExecution()
   Beard_Sides_Setup = RigVMFunctionReferenceNode(ExecuteContext=@RigUnit_BeginExecution.ExecutePin)
  ]}
]
hierarchy [
 root {loc=<0 0 0>, rot=<0 0 0 1>}
 pelvis {parent=root, loc=<0 -0.7071 93.1191>, rot=<0.0127 0.707 -0.0127 -0.707>}
 spine_01 {parent=pelvis, loc=<6.7494 -0.6388 0>, rot=<0 0 0.034 -0.9994>}
 spine_02 {parent=spine_01, loc=<7.2518 -0.035 0>, rot=<0 0 0.0238 -0.9997>}
 spine_03 {parent=spine_02, loc=<6.0104 0 0>, rot=<0 0 0.0149 -0.9999>}
 spine_04 {parent=spine_03, loc=<8.1024 0 0>, rot=<0 0 -0.0132 -0.9999>}
 spine_05 {parent=spine_04, loc=<10.0146 0 0>, rot=<0 0 -0.0015 -1>}
 neck_01 {parent=spine_05, loc=<11.3138 0.1529 0>, rot=<0 0 -0.2003 -0.9797>}
 neck_02 {parent=neck_01, loc=<4.5627 0 0>, rot=<0 0 0.0504 -0.9987>}
 head {parent=neck_02, loc=<5.3914 0 0>, rot=<0 0 0.115 -0.9934>}
 faceAttach {parent=head, loc=<-2.5 0.7625 0.015>, rot=<0.0082 -0.7071 -0.0082 -0.7071>}
 C_brow_mid {parent=faceAttach, loc=<-0.0073 9.0942 7.8054>, rot=<0 0.0066 -0.0013 -1>}
 L_brow_outer {parent=faceAttach, loc=<5.0345 6.2002 8.4399>, rot=<0.0018 -0.0017 0.4425 -0.8968>}
 L_cheek_inner {parent=faceAttach, loc=<4.4409 7.1051 3.406>, rot=<-0.1033 -0.3515 0.3052 -0.879>}
 C_lip_upper_mid {parent=faceAttach, loc=<0.0242 8.9897 0.5828>, rot=<0.2225 0.0022 -0.0008 -0.9749>}
 L_lip_upper_outer {parent=faceAttach, loc=<1.5912 8.3541 0.8324>, rot=<0.036 -0.0727 0.2627 -0.9615>}
 L_lip_corner {parent=faceAttach, loc=<2.9517 7.3567 0.5459>, rot=<0.0019 0.0091 -0.0145 -0.9999>}
 C_nose_bridge {parent=faceAttach, loc=<-0.0207 10.9594 4.7363>, rot=<-0.2218 0.0215 0.0061 -0.9748>}
 R_brow_outer {parent=faceAttach, loc=<-5.0095 6.1961 8.4207>, rot=<-0.0057 0.0111 -0.4442 -0.8958>}
 R_cheek_inner {parent=faceAttach, loc=<-4.4024 7.113 3.3972>, rot=<-0.1006 0.345 -0.3064 -0.8815>}
 R_lip_corner {parent=faceAttach, loc=<-2.8649 7.3695 0.5928>, rot=<0.0027 -0.0083 0.0146 -0.9999>}
 C_jaw {parent=faceAttach, loc=<0.0085 2.6932 1.1328>, rot=<0 0 0 1>}
 L_lip_lower_outer {parent=C_jaw, loc=<1.5223 5.6132 -0.5052>, rot=<0.0293 0.1335 0.3711 -0.9185>}
 C_lip_lower_mid {parent=C_jaw, loc=<0.0083 6.3105 -0.5919>, rot=<0 0.0182 0.0039 -0.9998>}
 R_lip_lower_outer {parent=C_jaw, loc=<-1.5468 5.8688 -0.563>, rot=<0.0224 -0.0957 -0.3438 -0.9339>}
 dyn_beard_01 {parent=C_jaw, loc=<0.0066 4.7777 -2.7825>, rot=<0.3645 -0.3645 -0.6059 -0.6059>}
 dyn_beard_02 {parent=dyn_beard_01, loc=<6.3939 0 0>, rot=<0 0 0 1>}
 dyn_beard_03 {parent=dyn_beard_02, loc=<6.1984 0 0>, rot=<0 0 0 1>}
 R_lip_upper_outer {parent=faceAttach, loc=<-1.5794 8.4129 0.8604>, rot=<0.0463 0.0435 -0.2515 -0.9658>}
 dyn_beard_01_l {parent=faceAttach, loc=<4.8732 6.8084 3.0863>, rot=<0.3069 -0.4277 0.0293 -0.8497>}
 dyn_beard_01_r {parent=faceAttach, loc=<-4.8431 6.8084 3.0865>, rot=<0.8497 0.0293 0.4277 0.3069>}
 dyn_hair_01 {parent=head, loc=<12.5982 -1.2231 6.4924>, rot=<0.5997 -0.3747 0.5966 -0.3796>}
 dyn_hair_02 {parent=head, loc=<8.765 -1.363 7.0165>, rot=<0.6083 -0.3605 0.6053 -0.3655>}
 dyn_hair_03 {parent=head, loc=<6.3242 -6.5947 4.0376>, rot=<0.6083 -0.3605 0.6053 -0.3655>}
 dyn_hair_04 {parent=head, loc=<13.8453 -4.3955 3.2412>, rot=<0.6083 -0.3605 0.6053 -0.3655>}
 dyn_hair_05 {parent=head, loc=<16.1596 -2.0287 1.7886>, rot=<0.583 -0.3999 0.58 -0.4047>}
 dyn_hair_06 {parent=head, loc=<16.4395 -2.3106 -1.246>, rot=<0.5323 -0.4655 0.5284 -0.4699>}
 dyn_hair_07 {parent=head, loc=<5.6932 -7.6759 2.0651>, rot=<0.5702 -0.4181 0.5667 -0.4228>}
 dyn_hair_08 {parent=head, loc=<6.7624 -7.5271 -0.343>, rot=<0.5176 -0.4818 0.5136 -0.4861>}
 dyn_hair_09 {parent=head, loc=<5.5651 -7.8404 -1.8095>, rot=<0.4267 -0.5639 0.422 -0.5674>}
 dyn_hair_010 {parent=head, loc=<10.3791 -8.7431 1.9968>, rot=<0.5504 -0.4439 0.5467 -0.4484>}
 dyn_hair_011 {parent=head, loc=<6.8081 -6.1259 -5.3498>, rot=<0.5578 -0.4345 0.5542 -0.4391>}
 dyn_hair_012 {parent=head, loc=<7.5148 -3.366 -6.7507>, rot=<0.5605 -0.431 0.557 -0.4357>}
 dyn_hair_013 {parent=head, loc=<14.7695 -0.3571 -4.5748>, rot=<0.4494 -0.5459 0.4449 -0.5496>}
 dyn_mustache_l_01 {parent=head, loc=<-0.6535 9.9986 -1.4794>, rot=<0.0725 -0.7034 0.0667 -0.704>}
 dyn_mustache_l_02 {parent=dyn_mustache_l_01, loc=<2.8298 -0.1297 -2.0502>, rot=<0 0 0.1807 -0.9835>}
 dyn_mustache_r_01 {parent=head, loc=<-0.6539 9.9986 1.4794>, rot=<-0.7034 -0.0725 0.704 0.0667>}
 dyn_mustache_r_02 {parent=dyn_mustache_r_01, loc=<-2.8298 0.1297 2.05>, rot=<0 0 0.1807 -0.9835>}
]
