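# Physics asset for the Tactical Santa Goggles hat accessory, defining capsule bodies and joints for its dynamic hair, beard and mustache.
# A beard chain hangs from the jaw, while two short mustache chains and separate single-bone hair bodies attach to the head with limited swing and twist.
# Jaw, head, upper spine and right hand capsules, plus a chest pouch shape, form body volumes linked by free-rotation constraints.
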PhysicsAsset /Game/Accessories/Hats/M_MED_Tactica_Santa_Goggles/Meshes/M_MED_Tactical_Santa_Goggles_Dyn
bodies {
 C_jaw [{shapeType=Capsule, length=2, radius=2}]
 dyn_beard_01 [{shapeType=Capsule, length=5.9316, radius=2.766, center=<1.8732 0 0>, rotation=<rot 0 0 90>}]
 dyn_beard_02 [{shapeType=Capsule, length=6.5, radius=2.766, center=<1.8732 -0.6643 0.6213>, rotation=<rot 0 0 90>}]
 dyn_beard_03 [{shapeType=Capsule, length=5.9316, radius=2.766, center=<1.8732 0 0>, rotation=<rot 0 0 90>}]
 dyn_hair_01 [{shapeType=Capsule, length=5, radius=3, center=<2 0 0>, rotation=<rot 90 0 0>}]
 dyn_hair_010 [{shapeType=Capsule, length=5, radius=3, center=<2 0 0>, rotation=<rot 90 0 0>}]
 dyn_hair_011 [{shapeType=Capsule, length=5, radius=3, center=<2 0 0>, rotation=<rot 90 0 0>}]
 dyn_hair_012 [{shapeType=Capsule, length=5, radius=3, center=<2 0 0>, rotation=<rot 90 0 0>}]
 dyn_hair_013 [{shapeType=Capsule, length=5, radius=3, center=<2 0 0>, rotation=<rot 90 0 0>}]
 dyn_hair_02 [{shapeType=Capsule, length=5, radius=3, center=<2 0 0>, rotation=<rot 90 0 0>}]
 dyn_hair_03 [{shapeType=Capsule, length=5, radius=3, center=<2 0 0>, rotation=<rot 90 0 0>}]
 dyn_hair_04 [{shapeType=Capsule, length=5, radius=3, center=<2 0 0>, rotation=<rot 90 0 0>}]
 dyn_hair_05 [{shapeType=Capsule, length=5, radius=3, center=<2 0 0>, rotation=<rot 90 0 0>}]
 dyn_hair_06 [{shapeType=Capsule, length=5, radius=3, center=<2 0 0>, rotation=<rot 90 0 0>}]
 dyn_hair_07 [{shapeType=Capsule, length=5, radius=3, center=<2 0 0>, rotation=<rot 90 0 0>}]
 dyn_hair_08 [{shapeType=Capsule, length=5, radius=3, center=<2 0 0>, rotation=<rot 90 0 0>}]
 dyn_hair_09 [{shapeType=Capsule, length=5, radius=3, center=<2 0 0>, rotation=<rot 90 0 0>}]
 dyn_mustache_l_01 [{shapeType=Capsule, length=2, radius=2, center=<-0.7097 0 0>, rotation=<rot 90 0 0>}]
 dyn_mustache_l_02 [{shapeType=Capsule, length=2, radius=2, center=<0.1 0 0>, rotation=<rot 90 0 0>}]
 dyn_mustache_r_01 [{shapeType=Capsule, length=2, radius=2, center=<0.7097 0 0>, rotation=<rot 90 0 0>}]
 dyn_mustache_r_02 [{shapeType=Capsule, length=2, radius=2, center=<-0.1 0 0>, rotation=<rot 90 0 0>}]
 hand_r [{shapeType=Capsule, length=13.0923, radius=4.8036, center=<-7.5392 -0.0001 4.9809>, rotation=<rot 90 0 0>}]
 head [{shapeType=Capsule, length=2, radius=2}]
 spine_03 [{shapeType=Capsule, length=13.8288, radius=10.8392, center=<9.0364 4.1674 0>, rotation=<rot 0 -90 0>}]
 spine_04 [{shapeType=Capsule, length=13.8686, radius=10.7993, center=<9.6619 4.7636 -0.0002>, rotation=<rot 0 -90 0>}]
 spine_05 [
  {shapeType=Capsule, length=20.1438, radius=15.695, center=<-2.2795 -1.6648 -0.3367>}
  {shapeType=Capsule, length=13.9518, radius=13.9425, center=<-0.2916 0.4452 0>}
  {
   shapeType=Capsule
   length=6.5
   radius=3
   shapeName=ChestPouch_capsule_0
   center=<4.4066 11.9469 -9.48>
   rotation=<rot -72.3105 -129.4699 105.0674>
  }
 ]
}
constraints [
 {bone1Name=dyn_beard_02, bone2Name=dyn_beard_01, swing1LimitDegrees=45, swing2LimitDegrees=45, twistLimitDegrees=15}
 {bone1Name=dyn_beard_01, bone2Name=C_jaw, swing1LimitDegrees=45, swing2LimitDegrees=45, twistLimitDegrees=15}
 {bone1Name=dyn_mustache_l_02, bone2Name=dyn_mustache_l_01, swing1LimitDegrees=15, swing2LimitDegrees=15, twistLimitDegrees=2}
 {bone1Name=dyn_mustache_r_02, bone2Name=dyn_mustache_r_01, swing1LimitDegrees=15, swing2LimitDegrees=15, twistLimitDegrees=2}
 {
  bone1Name=C_jaw
  bone2Name=head
  swing1LimitDegrees=45
  swing1Motion=Free
  swing2LimitDegrees=45
  swing2Motion=Free
  twistLimitDegrees=45
  twistMotion=Free
 }
 {bone1Name=dyn_hair_01, bone2Name=head, swing1LimitDegrees=45, swing2LimitDegrees=45, twistLimitDegrees=10}
 {bone1Name=dyn_hair_02, bone2Name=head, swing1LimitDegrees=45, swing2LimitDegrees=45, twistLimitDegrees=10}
 {bone1Name=dyn_hair_03, bone2Name=head, swing1LimitDegrees=45, swing2LimitDegrees=45, twistLimitDegrees=10}
 {bone1Name=dyn_hair_04, bone2Name=head, swing1LimitDegrees=45, swing2LimitDegrees=45, twistLimitDegrees=10}
 {bone1Name=dyn_hair_05, bone2Name=head, swing1LimitDegrees=45, swing2LimitDegrees=45, twistLimitDegrees=10}
 {bone1Name=dyn_hair_06, bone2Name=head, swing1LimitDegrees=45, swing2LimitDegrees=45, twistLimitDegrees=10}
 {bone1Name=dyn_hair_07, bone2Name=head, swing1LimitDegrees=45, swing2LimitDegrees=45, twistLimitDegrees=10}
 {bone1Name=dyn_hair_08, bone2Name=head, swing1LimitDegrees=45, swing2LimitDegrees=45, twistLimitDegrees=10}
 {bone1Name=dyn_hair_09, bone2Name=head, swing1LimitDegrees=45, swing2LimitDegrees=45, twistLimitDegrees=10}
 {bone1Name=dyn_hair_010, bone2Name=head, swing1LimitDegrees=45, swing2LimitDegrees=45, twistLimitDegrees=10}
 {bone1Name=dyn_hair_011, bone2Name=head, swing1LimitDegrees=45, swing2LimitDegrees=45, twistLimitDegrees=10}
 {bone1Name=dyn_hair_012, bone2Name=head, swing1LimitDegrees=45, swing2LimitDegrees=45, twistLimitDegrees=10}
 {bone1Name=dyn_hair_013, bone2Name=head, swing1LimitDegrees=45, swing2LimitDegrees=45, twistLimitDegrees=10}
 {bone1Name=dyn_mustache_l_01, bone2Name=head, swing1LimitDegrees=15, swing2LimitDegrees=15, twistLimitDegrees=2}
 {bone1Name=dyn_mustache_r_01, bone2Name=head, swing1LimitDegrees=15, swing2LimitDegrees=15, twistLimitDegrees=2}
 {
  bone1Name=head
  bone2Name=spine_05
  swing1LimitDegrees=45
  swing1Motion=Free
  swing2LimitDegrees=45
  swing2Motion=Free
  twistLimitDegrees=45
  twistMotion=Free
 }
 {
  bone1Name=spine_05
  bone2Name=spine_04
  swing1LimitDegrees=45
  swing1Motion=Free
  swing2LimitDegrees=45
  swing2Motion=Free
  twistLimitDegrees=45
  twistMotion=Free
 }
 {
  bone1Name=spine_04
  bone2Name=spine_03
  swing1LimitDegrees=45
  swing1Motion=Free
  swing2LimitDegrees=45
  swing2Motion=Free
  twistLimitDegrees=45
  twistMotion=Free
 }
 {
  bone1Name=hand_r
  bone2Name=spine_05
  swing1LimitDegrees=45
  swing1Motion=Free
  swing2LimitDegrees=45
  swing2Motion=Free
  twistLimitDegrees=45
  twistMotion=Free
 }
 {bone1Name=dyn_beard_03, bone2Name=dyn_beard_02, swing1LimitDegrees=45, swing2LimitDegrees=45, twistLimitDegrees=15}
]
